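; Gameplay ability that plays the spawn intro for the Baron AI character.
; After a short delay, it caches the avatar pawn, plays an intro montage and triggers a spawn speech bubble shortly into it.
; Montage completion or a fallback delay ends the ability through a DoOnce gate.
; On end, it adds a loose gameplay tag to the pawn if still valid.
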(event Ability|EventActivateAbility
  (bind (_montageinfo_montagetoplay _montageinfo_animplayrate _montageinfo_animrootmotiontranslationscale _montageinfo_montagesectiontoplay _montageinfo_overridesection) (Utilities|Struct|BreakFortGameplayAbilityMontageInfo (Variables|GetMontageInfo_MontageToPlay)))
  (Utilities|FlowControl|Delay 0.25)
  (bind _outfortplayerpawn (CallFunction|GetAvatarasValidFortPlayerPawn)
    (:ValidFortPlayerPawn
      (Variables|InternalData|SetAIPawn _outfortplayerpawn)
      (Ability|Tasks|PlayMontageAndWaitNotify "None" _montageinfo_montagetoplay _montageinfo_animplayrate _montageinfo_overridesection true _montageinfo_animrootmotiontranslationscale
        (:then
          (Utilities|FlowControl|Delay 0.2)
          (Class|BPFLAICommonFunctions|PlaySpeechBubble (Variables|InternalData|GetAIPawn) (Variables|Default|GetTag_OnSpawnedSpeech) true true)
          (Utilities|FlowControl|Delay 5.5) ; Fall back Delay
        )
        (:OnCompleted
          _))
      (Utilities|FlowControl|DoOnce false
        (:Completed
          (Ability|EndAbility))))))

(event Ability|EventFailedToActivatePassiveAbility
  (params (ActorInfo :type Gameplay%20Ability%20Actor%20Info%20Structure)))

(event Ability|EventOnEndAbility
  (params (bWasCancelled :type Boolean))
  (bind _aipawn (Variables|InternalData|GetAIPawn))
  (Utilities|IsValid _aipawn
    (:Is%20Valid
      (Ability|GameplayEffect|AddLooseGameplayTags _aipawn (Variables|Default|GetTagOnAbilityEnd) true))))
