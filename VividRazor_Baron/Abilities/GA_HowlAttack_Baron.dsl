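; Gameplay ability for the Baron's howl attack, built on a parent ability that handles activation.
; It overrides the custom ability source transform to sit at the top of the avatar pawn's capsule, adding the scaled half-height to the actor location.
; That transform keeps the pawn's rotation at unit scale.
; Activation only calls the parent, and the end-ability event adds no logic.
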(event Ability|EventActivateAbility
  (|CallParentFunction))

(event Ability|EventOnEndAbility
  (params (bWasCancelled :type Boolean)))

(fn GetCustomAbilitySourceTransform
  (return (Math|Vector|MakeVector (.x (Transformation|GetActorLocation (CallFunction|GetAvatarasFortPlayerPawn))) (.y (Transformation|GetActorLocation (CallFunction|GetAvatarasFortPlayerPawn))) (+ (.z (Transformation|GetActorLocation (CallFunction|GetAvatarasFortPlayerPawn))) (Components|Capsule|GetScaledCapsuleHalfHeight (Variables|/Script/Engine.Character|GetCapsuleComponent (CallFunction|GetAvatarasFortPlayerPawn))))) (Transformation|GetActorRotation (CallFunction|GetAvatarasFortPlayerPawn)) "1.000000,1.000000,1.000000"))
