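; Parent gameplay ability for an AI howl attack that detects players and applies gameplay effects to them.
; On activation it commits, then runs a target-selection pre-check and howls only if at least one actor is in range, otherwise ending.
; The howl plays a montage whose matching notify fires the howl cue and a pulse visual cue carrying the marking radius as magnitude.
; That notify also runs a second target selection to apply the effect container.
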(event Ability|EventActivateAbility
  (|CallParentFunction)
  (Ability|CommitAbility)
  ; Make sure we have at least 1 Target in Range before we do the actual Howl
  (bind (_asynctaskproxy _targetdata) (Ability|Tasks|DoTargetSelection "None" "(TagName=\"Targeting.ApplicationTag.1\")" true)
    (:Targeted
      (if (> (Utilities|Array|Length (Ability|TargetData|GetAllActorsfromTargetData _targetdata)) 0)
        (CallFunction|DoHowl)
        (else
          (Ability|EndAbility))))))

(event Custom|DoHowl
  (bind (_montageinfo_montagetoplay _montageinfo_animplayrate _montageinfo_animrootmotiontranslationscale _montageinfo_montagesectiontoplay _montageinfo_overridesection) (Utilities|Struct|BreakFortGameplayAbilityMontageInfo (Variables|GetMontageInfo_MontageToPlay)))
  ; Howl And Detect Players
  (bind (_asynctaskproxy _notifytag) (Ability|Tasks|PlayMontageAndWaitNotify "None" _montageinfo_montagetoplay _montageinfo_animplayrate _montageinfo_overridesection true _montageinfo_animrootmotiontranslationscale)
    (:OnCompleted
      _)
    (:OnInterrupted
      _)
    (:OnCancelled
      _)
    (:OnNotifyBegin
      (if (== _notifytag (Variables|Default|GetHowlEvent))
        (Ability|ExecuteGameplayCueOnOwner (Variables|Default|GetGC_Howling))
        (CallFunction|DoScanPulse)
        (bind (_asynctaskproxy_2 _targetdata _applicationtag) (Ability|Tasks|DoTargetSelection "None" "(TagName=\"Targeting.ApplicationTag.0\")" true)
          (:Targeted
            (Ability|ApplyGameplayEffectContainer _targetdata _applicationtag))))))
  (Ability|EndAbility))

(event Custom|DoScanPulse
  ; Trigger Pulse VFX
  (Ability|ExecuteGameplayCueWithParamsOnOwner (Variables|Default|GetGC_PulseVisual) (Ability|GameplayCue|MakeGameplayCueParameters :NormalizedMagnitude 0.0 :RawMagnitude (ScalableFloat|GetValueatLevel (Variables|Default|GetMarkingRadius)) :EffectContext (Class|AthenaFunctionLibrary|MakeSimpleFortEffectContext (Variables|InternalData|GetAIPawn) (Variables|InternalData|GetAIPawn)) :Location "0,0,0" :Normal "0,0,0" :Instigator (Variables|InternalData|GetAIPawn) :EffectCauser (Variables|InternalData|GetAIPawn) :SourceObject (Variables|InternalData|GetAIPawn) :GameplayEffectLevel 1 :AbilityLevel 1 :bReplicateLocationWhenUsingMinimalRepProxy false)))

(event Ability|EventOnEndAbility
  (params (bWasCancelled :type Boolean)))
