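; Gameplay ability for the reveal part of the VividRazor Baron's howl attack, run on an affected player pawn with authority.
; It skips non-participant players and finds the PulseIndicatorComponent on the Baron pawn taken from the effect context's source object.
; It points that component at the player's controller and team, then pulses markers for a single tick.
; It then applies a self-mark gameplay effect to the player and ends.
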(event Ability|EventActivateAbility
  (bind _playerpawn (Variables|Default|GetPlayerPawn))
  (bind _pulseindicator (Variables|Default|GetPulseIndicator))
  (Ability|HasAuthority
    (:True
      (bind _outfortplayerpawn (CallFunction|GetAvatarasValidFortPlayerPawn)
        (:ValidFortPlayerPawn
          (bind _output_get (Variables|Default|SetPlayerPawn _outfortplayerpawn))
          (if (not (GameplayTags|HasMatchingGameplayTag _output_get "(TagName=\"Pawn.Athena.NonParticipant\")"))
            (bind _asbp_player_pawn_non_participant_vivid_razor_baron (Utilities|Casting|CastToBP_PlayerPawn_NonParticipant_VividRazor_Baron (Ability|EffectContext|GetSourceObject (Ability|GetContextfromOwner))))
            (Utilities|IsValid (Actor|GetComponentbyClass _asbp_player_pawn_non_participant_vivid_razor_baron "/Script/FortniteGame.PulseIndicatorComponent")
              (:Is%20Valid
                (Variables|Default|SetPulseIndicator (Actor|GetComponentbyClass _asbp_player_pawn_non_participant_vivid_razor_baron "/Script/FortniteGame.PulseIndicatorComponent"))
                (bind _asfort_player_controller_athena (Utilities|Casting|CastToFortPlayerControllerAthena (Fortnite|GetFortPlayerControllerfromActor _playerpawn)))
                (Class|PulseIndicatorComponent|SetIndicatingController _pulseindicator _asfort_player_controller_athena)
                (Class|PulseIndicatorComponent|SetDisplayTeam _pulseindicator (Team|GetActorTeam _playerpawn))
                (Class|PulseIndicatorComponent|StartPulseMarkers (Variables|Default|GetPulseIndicator))
                (Utilities|FlowControl|DelayUntilNextTick)
                (Class|PulseIndicatorComponent|StopPulseMarkers (Variables|Default|GetPulseIndicator))
                (Creative|GameplayEffects|SimpleApplyGameplayEffectToActor (Variables|Default|GetGE_SelfMark) (Variables|Default|GetPlayerPawn))
                (Ability|EndAbility)))
            (else
              _)))
        (:AvatarCastFailed
          _))
      (Ability|EndAbility))))
