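; Gameplay ability for the VividRazor Baron boss that monitors a shield break and switches the AI into an enraged state.
; When tags are applied to the AI pawn, it runs the parent logic, then removes a tagged item from the bot controller's inventory.
; It then sets the enraged inventory and aiming skill sets on the Fortnite AI bot controller and broadcasts a stateful Baron event through the global event router.
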(event Ability|EventActivateAbility
  (|CallParentFunction))

(event Ability|EventOnEndAbility
  (params (bWasCancelled :type Boolean))
  (|CallParentFunction bWasCancelled) ; This node is disabled and will not be called.\nDrag off pins to build functionality.
)

(event EventApplyTagsToAIPawn
  (bind _aibotcontroller (Variables|InternalData|GetAIBotController))
  (|CallParentFunction)
  (bind _returnvalue (Inventory|GetInventoryItemwithTag _aibotcontroller (Variables|Default|GetTag%20to%20Find) true))
  (bind _asfort_world_item_definition (Utilities|Casting|CastToFortWorldItemDefinition (Item|GetItemDefinition _returnvalue)))
  (Inventory|RemoveItemfromInventoryOwner _aibotcontroller _asfort_world_item_definition)
  (Class|FortAthenaAIBotController|SetSkillSet _aibotcontroller (Variables|Default|GetInventorySkillSetWhenEnraged))
  (Class|FortAthenaAIBotController|SetSkillSet _aibotcontroller (Variables|Default|GetAimingSkillSetWhenEnraged))
  (GameplayEventRouter|BroadcastStatefulEvent (GameplayEventRouter|GetGlobalEventRouter) (Utilities|Struct|MakeVividRazorBaronEvent "NewEnumerator2") 0))
