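; Gameplay ability that starts the Baron AI's enrage with a speech bubble, montage and gameplay cue.
; On activation it calls the parent, then shows the enraged speech bubble on the AI pawn using a speech tag.
; It plays the configured montage through a montage task, firing the StartEnrage cue on the owner as the task starts.
; The ability ends when the montage completes, is interrupted or is cancelled.
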(event Ability|EventActivateAbility
  (bind (_montageinfo_montagetoplay _montageinfo_animplayrate _montageinfo_animrootmotiontranslationscale _montageinfo_montagesectiontoplay _montageinfo_overridesection) (Utilities|Struct|BreakFortGameplayAbilityMontageInfo (Variables|GetMontageInfo_MontageToPlay)))
  (|CallParentFunction)
  (Class|BPFLAICommonFunctions|PlaySpeechBubble (Variables|InternalData|GetAIPawn) (Variables|Default|GetTag_OnEnragedSpeech) true true)
  (Ability|Tasks|PlayMontageAndWaitNotify "None" _montageinfo_montagetoplay _montageinfo_animplayrate _montageinfo_overridesection true _montageinfo_animrootmotiontranslationscale
    (:then
      (Ability|ExecuteGameplayCueOnOwner (Variables|Default|GetGC_StartEnrage)))
    (:OnCompleted
      _)
    (:OnInterrupted
      _)
    (:OnCancelled
      _))
  (Ability|EndAbility))
