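; Parent gameplay ability for the Baron's triple claw attack, driven by a montage.
; On activation it commits the ability and plays the configured montage through a wait-for-trigger task.
; Each trigger applies the gameplay effect container to the supplied target data and tag, then executes the attacking gameplay cue on the owner.
; Montage completion ends the ability, and the end handler always passes false to the parent as the cancelled flag.
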(event Ability|EventActivateAbility
  (bind (_montageinfo_montagetoplay _montageinfo_animplayrate _montageinfo_animrootmotiontranslationscale _montageinfo_montagesectiontoplay _montageinfo_overridesection) (Utilities|Struct|BreakFortGameplayAbilityMontageInfo (Variables|GetMontageInfo_MontageToPlay)))
  (|CallParentFunction)
  (Ability|CommitAbility)
  (bind (_asynctaskproxy _targetdata _applicationtag) (Ability|Tasks|PlayMontageandWaitforTrigger "None" _montageinfo_montagetoplay _montageinfo_animplayrate _montageinfo_montagesectiontoplay _montageinfo_overridesection _montageinfo_animrootmotiontranslationscale)
    (:Triggered
      (Ability|ApplyGameplayEffectContainer _targetdata _applicationtag)
      (Ability|ExecuteGameplayCueOnOwner (Variables|Default|GetGC_Attacking)))
    (:Completed
      (Ability|EndAbility))))

(event Ability|EventOnEndAbility
  (params (bWasCancelled :type Boolean))
  (|CallParentFunction false))
